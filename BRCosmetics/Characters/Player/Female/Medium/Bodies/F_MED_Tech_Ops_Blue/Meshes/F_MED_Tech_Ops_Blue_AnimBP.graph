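# Post-process style AnimBP for the Tech Ops Blue female body, adding secondary motion to bullets, pouches, straps and shoelaces.
# It copies the parent mesh pose, then chains constraints, look-ats, bone offsets and AnimDynamics.
# Pouches and straps simulate relative to a 'sim' bone that follows only pelvis translation.
# EventGraph derives crouch and left-thigh corrective alphas, and swaps strap gravity for scaled skydiving force while skydiving.
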AnimBlueprint /BRCosmetics/Characters/Player/Female/Medium/Bodies/F_MED_Tech_Ops_Blue/Meshes/F_MED_Tech_Ops_Blue_AnimBP
target_skeleton=&/F_MED_Tech_Ops_Blue_Skeleton
graphs [
 AnimGraph {nodes [
   n1 = "Constraint - sim (T)"(ComponentPose=Local To Component(LocalPose=Copy Pose From Mesh() {bUseAttachedParent=true}), Alpha=1, ConstraintWeights_0=1) {
    alphaBoolBlend {blendOption=Linear}
    boneToModify {boneName=sim}
    constraintSetup [{transformType=Translation, targetBone {boneName=pelvis}}]
    constraintWeights [1]
   }
   n2 = Anim Dynamics - Bone: dyn_bullet_bk_04_r(ComponentPose=@n1, Alpha=1, ExternalForce="0,0,0") {
    angularDampingOverride=1
    angularSpringConstant=1500
    bAngularSpring=true
    bGravityOverrideInSimSpace=true
    bOverrideAngularDamping=true
    gravityScale=0.35
    simulationSpace=RootRelative
    alphaBoolBlend {blendOption=Linear}
    boundBone {boneName=dyn_bullet_bk_04_r}
    componentAppliedLinearAccClamp=<3000 3000 3000>
    componentLinearAccScale=<0.3 0.3 0.3>
    componentLinearVelScale=<0.3 0.3 0.1>
    physicsBodyDefinitions [{
      sphereCollisionRadius=0
      boundBone {boneName=dyn_bullet_bk_04_r}
      boxExtents=<10 20 20>
      constraintSetup {angularLimitsMax=<5 10 5>, angularLimitsMin=<-5 -20 -5>, angularTarget=<0 0 0>}
      localJointOffset=<10 0 0>
     }]
   }
   n3 = Anim Dynamics - Bone: dyn_bullet_bk_03_r(ComponentPose=@n2, Alpha=1, ExternalForce="0,0,0") {
    angularDampingOverride=1
    angularSpringConstant=1500
    bAngularSpring=true
    bGravityOverrideInSimSpace=true
    bOverrideAngularDamping=true
    gravityScale=0.35
    simulationSpace=RootRelative
    alphaBoolBlend {blendOption=Linear}
    boundBone {boneName=dyn_bullet_bk_03_r}
    componentAppliedLinearAccClamp=<3000 3000 3000>
    componentLinearAccScale=<0.3 0.3 0.3>
    componentLinearVelScale=<0.3 0.3 1.5>
    physicsBodyDefinitions [{
      sphereCollisionRadius=0
      boundBone {boneName=dyn_bullet_bk_03_r}
      boxExtents=<10 20 20>
      constraintSetup {angularLimitsMax=<5 6 5>, angularLimitsMin=<-5 -20 -5>, angularTarget=<0 0 0>}
      localJointOffset=<10 0 0>
     }]
   }
   n4 = Anim Dynamics - Bone: dyn_bullet_bk_02_r(ComponentPose=@n3, Alpha=1, ExternalForce="0,0,0") {
    angularDampingOverride=1
    angularSpringConstant=1500
    bAngularSpring=true
    bGravityOverrideInSimSpace=true
    bOverrideAngularDamping=true
    gravityScale=0.35
    simulationSpace=RootRelative
    alphaBoolBlend {blendOption=Linear}
    boundBone {boneName=dyn_bullet_bk_02_r}
    componentAppliedLinearAccClamp=<3000 3000 3000>
    componentLinearAccScale=<0.3 0.3 0.3>
    componentLinearVelScale=<0.3 0.3 4>
    physicsBodyDefinitions [{
      sphereCollisionRadius=0
      boundBone {boneName=dyn_bullet_bk_02_r}
      boxExtents=<10 20 20>
      constraintSetup {angularLimitsMax=<5 5 5>, angularLimitsMin=<-5 -20 -5>, angularTarget=<0 0 0>}
      localJointOffset=<10 0 0>
     }]
   }
   n5 = Anim Dynamics - Bone: dyn_bullet_bk_01_r(ComponentPose=@n4, Alpha=1, ExternalForce="0,0,0") {
    angularDampingOverride=1
    angularSpringConstant=1500
    bAngularSpring=true
    bGravityOverrideInSimSpace=true
    bOverrideAngularDamping=true
    gravityScale=0.35
    simulationSpace=RootRelative
    alphaBoolBlend {blendOption=Linear}
    boundBone {boneName=dyn_bullet_bk_01_r}
    componentAppliedLinearAccClamp=<3000 3000 3000>
    componentLinearAccScale=<0.3 0.3 0.3>
    componentLinearVelScale=<0.3 0.3 2>
    physicsBodyDefinitions [{
      sphereCollisionRadius=0
      boundBone {boneName=dyn_bullet_bk_01_r}
      boxExtents=<10 20 20>
      constraintSetup {angularLimitsMax=<5 15 5>, angularLimitsMin=<-5 -20 -5>, angularTarget=<0 0 0>}
      localJointOffset=<10 0 0>
     }]
   }
   n6 = Anim Dynamics - Bone: dyn_bullet_fr_01_l(ComponentPose=@n5, Alpha=1, ExternalForce="0,0,0") {
    angularDampingOverride=1
    angularSpringConstant=1500
    bAngularSpring=true
    bGravityOverrideInSimSpace=true
    bOverrideAngularDamping=true
    gravityScale=0.35
    simulationSpace=RootRelative
    alphaBoolBlend {blendOption=Linear}
    boundBone {boneName=dyn_bullet_fr_01_l}
    componentAppliedLinearAccClamp=<3000 3000 1250>
    componentLinearAccScale=<0.3 0.3 0.3>
    componentLinearVelScale=<0.3 0.3 1.7>
    physicsBodyDefinitions [{
      sphereCollisionRadius=0
      boundBone {boneName=dyn_bullet_fr_01_l}
      boxExtents=<10 20 20>
      constraintSetup {angularLimitsMax=<2.5 5 2.5>, angularLimitsMin=<-2.5 -3 -2.5>, angularTarget=<0 0 0>}
      localJointOffset=<10 0 0>
     }]
   }
   n7 = Anim Dynamics - Bone: dyn_bullet_fr_02_l(ComponentPose=@n6, Alpha=1, ExternalForce="0,0,0") {
    angularDampingOverride=1
    angularSpringConstant=1500
    bAngularSpring=true
    bGravityOverrideInSimSpace=true
    bOverrideAngularDamping=true
    gravityScale=0.35
    simulationSpace=RootRelative
    alphaBoolBlend {blendOption=Linear}
    boundBone {boneName=dyn_bullet_fr_02_l}
    componentAppliedLinearAccClamp=<3000 3000 1250>
    componentLinearAccScale=<0.3 0.3 0.3>
    componentLinearVelScale=<0.3 0.3 2>
    physicsBodyDefinitions [{
      sphereCollisionRadius=0
      boundBone {boneName=dyn_bullet_fr_02_l}
      boxExtents=<10 20 20>
      constraintSetup {angularLimitsMax=<2.5 5 2.5>, angularLimitsMin=<-2.5 -3 -2.5>, angularTarget=<0 0 0>}
      localJointOffset=<10 0 0>
     }]
   }
   n8 = Anim Dynamics - Bone: dyn_bullet_fr_03_l(ComponentPose=@n7, Alpha=1, ExternalForce="0,0,0") {
    angularDampingOverride=1
    angularSpringConstant=150
    bAngularSpring=true
    bGravityOverrideInSimSpace=true
    bOverrideAngularDamping=true
    gravityScale=0.35
    simulationSpace=RootRelative
    alphaBoolBlend {blendOption=Linear}
    boundBone {boneName=dyn_bullet_fr_03_l}
    componentAppliedLinearAccClamp=<3000 3000 1250>
    componentLinearAccScale=<0.3 0.3 0.3>
    componentLinearVelScale=<0.3 0.3 1.4>
    physicsBodyDefinitions [{
      sphereCollisionRadius=0
      boundBone {boneName=dyn_bullet_fr_03_l}
      boxExtents=<10 20 20>
      constraintSetup {angularLimitsMax=<2.5 2 2.5>, angularLimitsMin=<-2.5 -2 -2.5>, angularTarget=<0 0 0>}
      localJointOffset=<10 0 0>
     }]
   }
   n9 = Get thighCrouchDelta_l() -> (thighCrouchDelta_l:double)
   n10 = "Transform (Modify) Bone - Bone: dyn_bullet_fr_01_l"(ComponentPose=@n8, Alpha=@n9, Translation="-0.113,3.036,-0.242", Rotation="51.709,53.794,46.996", Scale="1,1,1") {
    rotationMode=BMM_Additive
    rotationSpace=BCS_BoneSpace
    translationMode=BMM_Additive
    translationSpace=BCS_BoneSpace
    alphaBoolBlend {blendOption=Linear}
    boneToModify {boneName=dyn_bullet_fr_01_l}
    rotation=<rot 51.7089 53.794 46.9959>
    translation=<-0.1132 3.0357 -0.2424>
   }
   n11 = "Transform (Modify) Bone - Bone: dyn_bullet_fr_02_l"(ComponentPose=@n10, Alpha=@n9, Translation="0.000,3.044,-0.000", Rotation="53.775,76.059,72.911", Scale="1,1,1") {
    rotationMode=BMM_Additive
    rotationSpace=BCS_BoneSpace
    translationMode=BMM_Additive
    translationSpace=BCS_BoneSpace
    alphaBoolBlend {blendOption=Linear}
    boneToModify {boneName=dyn_bullet_fr_02_l}
    rotation=<rot 53.7749 76.0587 72.9108>
    translation {y=3.044}
   }
   n12 = "Transform (Modify) Bone - Bone: dyn_bullet_fr_03_l"(ComponentPose=@n11, Alpha=@n9, Translation="0.000,3.044,-0.000", Rotation="55.504,101.959,94.155", Scale="1,1,1") {
    rotationMode=BMM_Additive
    rotationSpace=BCS_BoneSpace
    translationMode=BMM_Additive
    translationSpace=BCS_BoneSpace
    alphaBoolBlend {blendOption=Linear}
    boneToModify {boneName=dyn_bullet_fr_03_l}
    rotation=<rot 55.5039 101.9592 94.1549>
    translation {y=3.044}
   }
   n13 = "Constraint - dyn_pouch_02_r (R)"(ComponentPose=@n12, Alpha=1, ConstraintWeights_0=1) {
    alphaBoolBlend {blendOption=Linear}
    boneToModify {boneName=dyn_pouch_02_r}
    constraintSetup [{transformType=Rotation, targetBone {boneName=thigh_r}}]
    constraintWeights [1]
   }
   n14 = "Constraint - dyn_pouch_01_r (R)"(ComponentPose=@n13, Alpha=1, ConstraintWeights_0=1) {
    alphaBoolBlend {blendOption=Linear}
    boneToModify {boneName=dyn_pouch_01_r}
    constraintSetup [{transformType=Rotation, targetBone {boneName=thigh_r}}]
    constraintWeights [1]
   }
   n15 = "Transform (Modify) Bone - Bone: dyn_pouch_01_l"(
    ComponentPose=@n14
    Alpha=Get isCrouchingAlpha() -> (isCrouchingAlpha:double)
    Translation="0,0,0"
    Rotation="-0.020,0.004,-35"
    Scale="1,1,1"
   ) {
    rotationMode=BMM_Additive
    rotationSpace=BCS_BoneSpace
    scaleSpace=BCS_BoneSpace
    translationMode=BMM_Additive
    translationSpace=BCS_BoneSpace
    alphaBoolBlend {blendOption=Linear}
    boneToModify {boneName=dyn_pouch_01_l}
    rotation=<rot -0.02 0.004 -35>
   }
   n16 = Anim Dynamics - Bone: dyn_pouch_01_l(ComponentPose=@n15, Alpha=1, ExternalForce="0,0,0") {
    angularDampingOverride=1
    angularSpringConstant=1500
    bAngularSpring=true
    bGravityOverrideInSimSpace=true
    bOverrideAngularDamping=true
    bUseGravityOverride=true
    gravityScale=0.35
    simulationSpace=BoneRelative
    alphaBoolBlend {blendOption=Linear}
    boundBone {boneName=dyn_pouch_01_l}
    componentAppliedLinearAccClamp=<3000 3000 3000>
    componentLinearAccScale=<0.3 0.3 0.3>
    componentLinearVelScale=<0.3 0.3 2>
    gravityOverride {z=-980}
    physicsBodyDefinitions [{
      sphereCollisionRadius=0
      boundBone {boneName=dyn_pouch_01_l}
      boxExtents=<20 10 10>
      constraintSetup {angularLimitsMax=<5 0 5>, angularLimitsMin=<-5 -45 -5>, angularTarget=<0 0 0>}
      localJointOffset=<4 0 0>
     }]
    relativeSpaceBone {boneName=sim}
   }
   n17 = Anim Dynamics - Bone: dyn_pouch_01_r(ComponentPose=@n16, Alpha=1, ExternalForce="0,0,0") {
    angularDampingOverride=1
    angularSpringConstant=1500
    bAngularSpring=true
    bGravityOverrideInSimSpace=true
    bOverrideAngularDamping=true
    bUseGravityOverride=true
    gravityScale=0.35
    simulationSpace=BoneRelative
    alphaBoolBlend {blendOption=Linear}
    boundBone {boneName=dyn_pouch_01_r}
    componentAppliedLinearAccClamp=<3000 3000 3000>
    componentLinearAccScale=<0.3 0.3 0.3>
    componentLinearVelScale=<0.3 0.3 2>
    gravityOverride {z=-980}
    physicsBodyDefinitions [{
      sphereCollisionRadius=0
      boundBone {boneName=dyn_pouch_01_r}
      boxExtents=<20 10 10>
      constraintSetup {angularLimitsMax=<5 0 5>, angularLimitsMin=<-5 -45 -5>, angularTarget=<0 0 0>}
      localJointOffset=<4 0 0>
     }]
    relativeSpaceBone {boneName=sim}
   }
   n18 = Anim Dynamics - Bone: dyn_pouch_02_r(ComponentPose=@n17, Alpha=1, ExternalForce="0,0,0") {
    angularDampingOverride=1
    angularSpringConstant=1500
    bAngularSpring=true
    bGravityOverrideInSimSpace=true
    bOverrideAngularDamping=true
    bUseGravityOverride=true
    gravityScale=0.35
    simulationSpace=BoneRelative
    alphaBoolBlend {blendOption=Linear}
    boundBone {boneName=dyn_pouch_02_r}
    componentAppliedLinearAccClamp=<3000 3000 3000>
    componentLinearAccScale=<0.3 0.3 0.3>
    componentLinearVelScale=<0.3 0.3 2>
    gravityOverride {z=-980}
    physicsBodyDefinitions [{
      boundBone {boneName=dyn_pouch_02_r}
      boxExtents=<20 10 10>
      constraintSetup {angularLimitsMax=<5 0 5>, angularLimitsMin=<-5 -45 -5>, angularTarget=<0 0 0>}
      localJointOffset=<4 0 0>
     }]
    relativeSpaceBone {boneName=sim}
    sphericalLimits [{limitRadius=10, drivingBone {boneName=thigh_r}, sphereLocalOffset=<10 10 0>}]
   }
   n19 = Anim Dynamics - Bone: dyn_pouch_03_r(ComponentPose=@n18, Alpha=1, ExternalForce="0,0,0") {
    angularDampingOverride=1
    angularSpringConstant=1500
    bAngularSpring=true
    bGravityOverrideInSimSpace=true
    bOverrideAngularDamping=true
    bUseGravityOverride=true
    gravityScale=0.35
    simulationSpace=BoneRelative
    alphaBoolBlend {blendOption=Linear}
    boundBone {boneName=dyn_pouch_03_r}
    componentAppliedLinearAccClamp=<3000 3000 3000>
    componentLinearAccScale=<0.3 0.3 0.3>
    componentLinearVelScale=<0.3 0.3 2>
    gravityOverride {z=-980}
    physicsBodyDefinitions [{
      sphereCollisionRadius=0
      boundBone {boneName=dyn_pouch_03_r}
      boxExtents=<20 10 10>
      constraintSetup {angularLimitsMax=<5 0 5>, angularLimitsMin=<-5 -45 -5>, angularTarget=<0 0 0>}
      localJointOffset=<4 0 0>
     }]
    relativeSpaceBone {boneName=sim}
   }
   n20 = "Constraint - dyn_straps_bk_aim (PPPP)"!(
    ComponentPose=@n19
    Alpha=1
    ConstraintWeights_0=0.6911
    ConstraintWeights_1:float=0.2639
    ConstraintWeights_2:float=0.0071
    ConstraintWeights_3:float=0.0378
   ) -> (Pose) {
    alphaBoolBlend {blendOption=Linear}
    boneToModify {boneName=dyn_straps_bk_aim}
    constraintSetup [
     {targetBone {boneName=pelvis}}
     {targetBone {boneName=thigh_twist_01_l}}
     {targetBone {boneName=thigh_twist_01_r}}
     {targetBone {boneName=thigh_l}}
    ]
    constraintWeights [1,1,1,1]
   }
   n21 = "Constraint - dyn_straps_fr_aim (PPPP)"!(
    ComponentPose=@n20
    Alpha=1
    ConstraintWeights_0=0.2703
    ConstraintWeights_1:float=0.6534
    ConstraintWeights_2:float=0.0021
    ConstraintWeights_3:float=0.0742
   ) -> (Pose) {
    alphaBoolBlend {blendOption=Linear}
    boneToModify {boneName=dyn_straps_fr_aim}
    constraintSetup [
     {targetBone {boneName=pelvis}}
     {targetBone {boneName=thigh_twist_01_l}}
     {targetBone {boneName=thigh_twist_01_r}}
     {targetBone {boneName=thigh_l}}
    ]
    constraintWeights [1,1,1,1]
   }
   n22 = Look At - Bone: dyn_straps_fr_01(ComponentPose=@n21, Alpha=1) {
    alphaBoolBlend {blendOption=Linear}
    boneToModify {boneName=dyn_straps_fr_01}
    lookAtLocation {x=0}
    lookAtTarget {boneReference {boneName=dyn_straps_fr_aim}}
    lookAt_Axis {axis=<1 0>}
    lookUp_Axis {axis {x=0, z=1}}
   }
   n23 = Look At - Bone: dyn_straps_bk_01(ComponentPose=@n22, Alpha=1) {
    alphaBoolBlend {blendOption=Linear}
    boneToModify {boneName=dyn_straps_bk_01}
    lookAtLocation {x=0}
    lookAtTarget {boneReference {boneName=dyn_straps_bk_aim}}
    lookAt_Axis {axis=<1 0>}
    lookUp_Axis {axis {x=0, z=1}}
   }
   n24 = "Transform (Modify) Bone - Bone: dyn_straps_fr_01"(ComponentPose=@n23, Alpha=1, Translation="1.596,-0.071,-0.007", Rotation="0,0,0", Scale="1,1,1") {
    rotationMode=BMM_Additive
    rotationSpace=BCS_BoneSpace
    translationMode=BMM_Additive
    translationSpace=BCS_BoneSpace
    alphaBoolBlend {blendOption=Linear}
    boneToModify {boneName=dyn_straps_fr_01}
    translation=<1.596 -0.071 -0.007>
   }
   n25 = "Transform (Modify) Bone - Bone: dyn_strap_01"(ComponentPose=@n24, Alpha=1, Translation="0.499,-0.069,-0.007", Rotation="0.000,-5.000,-0.001", Scale="1,1,1") {
    rotationMode=BMM_Additive
    rotationSpace=BCS_BoneSpace
    translationMode=BMM_Additive
    translationSpace=BCS_BoneSpace
    alphaBoolBlend {blendOption=Linear}
    boneToModify {boneName=dyn_strap_01}
    rotation {roll=-0.001, yaw=-5.0001}
    translation=<0.499 -0.069 -0.007>
   }
   n26 = "Transform (Modify) Bone - Bone: dyn_straps_bk_01"(ComponentPose=@n25, Alpha=1, Translation="0,0,0", Rotation="14.985,0.089,0.670", Scale="1,1,1") {
    rotationMode=BMM_Additive
    rotationSpace=BCS_BoneSpace
    translationMode=BMM_Additive
    translationSpace=BCS_BoneSpace
    alphaBoolBlend {blendOption=Linear}
    boneToModify {boneName=dyn_straps_bk_01}
    rotation=<rot 14.985 0.089 0.67>
   }
   n27 = "Transform (Modify) Bone - Bone: dyn_strap_02"(ComponentPose=@n26, Alpha=0.5, Translation="-1.448,-0.191,0.254", Rotation="-9.908,-0.176,1.354", Scale="1,1,1") {
    rotationMode=BMM_Additive
    rotationSpace=BCS_BoneSpace
    translationMode=BMM_Additive
    translationSpace=BCS_BoneSpace
    alphaBoolBlend {blendOption=Linear}
    boneToModify {boneName=dyn_strap_02}
    rotation=<rot -9.908 -0.176 1.354>
    translation=<-1.448 -0.191 0.254>
   }
   n28 = "Transform (Modify) Bone - Bone: dyn_straps_bk_02"(ComponentPose=@n27, Alpha=1, Translation="-1.448,-0.191,0.254", Rotation="-9.908,-0.176,1.354", Scale="1,1,1") {
    rotationMode=BMM_Additive
    rotationSpace=BCS_BoneSpace
    translationMode=BMM_Additive
    translationSpace=BCS_BoneSpace
    alphaBoolBlend {blendOption=Linear}
    boneToModify {boneName=dyn_straps_bk_02}
    rotation=<rot -9.908 -0.176 1.354>
    translation=<-1.448 -0.191 0.254>
   }
   n29 = "Constraint - dyn_strap_aim (PPP)"!(ComponentPose=@n28, Alpha=1, ConstraintWeights_0=0.1132, ConstraintWeights_1:float=0.5054, ConstraintWeights_2:float=0.3814) -> (Pose) {
    alphaBoolBlend {blendOption=Linear}
    boneToModify {boneName=dyn_strap_aim}
    constraintSetup [{targetBone {boneName=pelvis}}, {targetBone {boneName=thigh_twist_01_l}}, {targetBone {boneName=thigh_l}}]
    constraintWeights [1,1,1]
   }
   AnimGraphNode_AnimDynamics_0 = Look At - Bone: dyn_strap_01(ComponentPose=@n29, Alpha=1) {
    alphaBoolBlend {blendOption=Linear}
    boneToModify {boneName=dyn_strap_01}
    lookAtLocation {x=0}
    lookAtTarget {boneReference {boneName=dyn_strap_aim}}
    lookAt_Axis {axis=<1 0>}
    lookUp_Axis {axis {x=0, z=1}}
   }
   n30 = "Constraint - dyn_strapsbuckle_fr_01 (RR)"!(ComponentPose=@AnimGraphNode_AnimDynamics_0, Alpha=1, ConstraintWeights_0=0.5, ConstraintWeights_1:float=0.5) -> (Pose) {
    alphaBoolBlend {blendOption=Linear}
    boneToModify {boneName=dyn_strapsbuckle_fr_01}
    constraintSetup [{transformType=Rotation, targetBone {boneName=pelvis}}, {transformType=Rotation, targetBone {boneName=dyn_straps_fr_01}}]
    constraintWeights [1,1]
   }
   n31 = Get windVector() -> (windVector:Vector)
   n32 = Anim Dynamics - Bone: dyn_strap_01!(ComponentPose=@n30, Alpha=1, GravityOverride:Vector=@n31, ExternalForce="0,0,0") -> (Pose) {
    angularDampingOverride=1
    angularSpringConstant=1500
    bAngularSpring=true
    bGravityOverrideInSimSpace=true
    bOverrideAngularDamping=true
    bUseGravityOverride=true
    gravityScale=0.35
    simulationSpace=BoneRelative
    alphaBoolBlend {blendOption=Linear}
    boundBone {boneName=dyn_strap_01}
    componentAppliedLinearAccClamp=<3000 3000 3000>
    componentLinearAccScale=<0.3 0.3 0.3>
    componentLinearVelScale=<0.3 0.3 2>
    gravityOverride {z=-980}
    physicsBodyDefinitions [{
      sphereCollisionRadius=0
      boundBone {boneName=dyn_strap_01}
      boxExtents=<20 10 10>
      constraintSetup {angularLimitsMax=<2.5 0 2.5>, angularLimitsMin=<-2.5 -20 -2.5>, angularTarget=<0 0 0>}
      localJointOffset=<5 0 0>
     }]
    relativeSpaceBone {boneName=sim}
   }
   n33 = Anim Dynamics - Bone: dyn_straps_bk_01!(ComponentPose=@n32, Alpha=1, GravityOverride:Vector=@n31, ExternalForce="0,0,0") -> (Pose) {
    angularDampingOverride=1
    angularSpringConstant=150
    bAngularSpring=true
    bGravityOverrideInSimSpace=true
    bOverrideAngularDamping=true
    bUseGravityOverride=true
    gravityScale=0.35
    simulationSpace=BoneRelative
    alphaBoolBlend {blendOption=Linear}
    boundBone {boneName=dyn_straps_bk_01}
    componentAppliedLinearAccClamp=<3000 3000 1250>
    componentLinearAccScale=<0.3 0.3>
    componentLinearVelScale=<0.3 0.3 4>
    gravityOverride {z=-980}
    physicsBodyDefinitions [{
      sphereCollisionRadius=0
      boundBone {boneName=dyn_straps_bk_01}
      boxExtents=<20 10 10>
      constraintSetup {angularLimitsMax=<2.5 1 2.5>, angularLimitsMin=<-2.5 -45 -2.5>, angularTarget=<0 0 0>}
      localJointOffset=<3 0 0>
     }]
    relativeSpaceBone {boneName=sim}
   }
   n34 = Anim Dynamics - Bone: dyn_straps_bk_01_r!(ComponentPose=@n33, Alpha=1, GravityOverride:Vector=@n31, ExternalForce="0,0,0") -> (Pose) {
    angularDampingOverride=1
    angularSpringConstant=150
    bAngularSpring=true
    bGravityOverrideInSimSpace=true
    bOverrideAngularDamping=true
    bUseGravityOverride=true
    gravityScale=0.35
    simulationSpace=BoneRelative
    alphaBoolBlend {blendOption=Linear}
    boundBone {boneName=dyn_straps_bk_01_r}
    componentAppliedLinearAccClamp=<3000 3000 1250>
    componentLinearAccScale=<0.3 0.3 0.3>
    componentLinearVelScale=<0.3 0.3 4>
    gravityOverride {z=-980}
    physicsBodyDefinitions [{
      sphereCollisionRadius=0
      boundBone {boneName=dyn_straps_bk_01_r}
      boxExtents=<20 10 10>
      constraintSetup {angularLimitsMax=<2.5 1 2.5>, angularLimitsMin=<-2.5 -25 -2.5>, angularTarget=<0 0 0>}
      localJointOffset=<3 0 0>
     }]
    relativeSpaceBone {boneName=sim}
   }
   n35 = Anim Dynamics - Bone: dyn_straps_fr_01!(ComponentPose=@n34, Alpha=1, GravityOverride:Vector=@n31, ExternalForce="0,0,0") -> (Pose) {
    angularDampingOverride=1
    angularSpringConstant=1500
    bAngularSpring=true
    bGravityOverrideInSimSpace=true
    bOverrideAngularDamping=true
    bUseGravityOverride=true
    gravityScale=0.35
    simulationSpace=BoneRelative
    alphaBoolBlend {blendOption=Linear}
    boundBone {boneName=dyn_straps_fr_01}
    componentAppliedLinearAccClamp=<3000 3000 3000>
    componentLinearAccScale=<0.3 0.3 0.3>
    componentLinearVelScale=<0.3 0.3 2>
    gravityOverride {z=-980}
    physicsBodyDefinitions [{
      sphereCollisionRadius=0
      boundBone {boneName=dyn_straps_fr_01}
      boxExtents=<20 10 10>
      constraintSetup {angularLimitsMax=<2.5 0 2.5>, angularLimitsMin=<-2.5 -5 -2.5>, angularTarget=<0 0 0>}
      localJointOffset=<3 0 0>
     }]
    relativeSpaceBone {boneName=sim}
   }
   n36 = Anim Dynamics - Bone: dyn_strapsbuckle_fr_01!(ComponentPose=@n35, Alpha=1, GravityOverride:Vector=@n31, ExternalForce="0,0,0") -> (Pose) {
    angularDampingOverride=1
    angularSpringConstant=1500
    bAngularSpring=true
    bGravityOverrideInSimSpace=true
    bOverrideAngularDamping=true
    bUseGravityOverride=true
    gravityScale=0.35
    simulationSpace=BoneRelative
    alphaBoolBlend {blendOption=Linear}
    boundBone {boneName=dyn_strapsbuckle_fr_01}
    componentAppliedLinearAccClamp=<3000 3000 3000>
    componentLinearAccScale=<0.3 0.3 0.3>
    componentLinearVelScale=<0.3 0.3 2>
    gravityOverride {z=-980}
    physicsBodyDefinitions [{
      sphereCollisionRadius=0
      boundBone {boneName=dyn_strapsbuckle_fr_01}
      boxExtents=<20 10 10>
      constraintSetup {angularLimitsMax=<2.5 0 2.5>, angularLimitsMin=<-2.5 -5 -2.5>, angularTarget=<0 0 0>}
      localJointOffset=<6 0 0>
     }]
    relativeSpaceBone {boneName=sim}
   }
   n37 = "Transform (Modify) Bone - Bone: dyn_strap_02"(
    ComponentPose=@n36
    Alpha=Get thighCrouchDelta_slide_l() -> (thighCrouchDelta_slide_l:double)
    Translation="0.000,-6.419,3.943"
    Rotation="0,0,0"
    Scale="1,1,1"
   ) {
    rotationMode=BMM_Additive
    rotationSpace=BCS_BoneSpace
    translationMode=BMM_Additive
    translationSpace=BCS_BoneSpace
    alphaBoolBlend {blendOption=Linear}
    boneToModify {boneName=dyn_strap_02}
    translation=<0.0001 -6.419 3.9429>
   }
   n38 = "Transform (Modify) Bone - Bone: dyn_straps_bk_02"(ComponentPose=@n37, Alpha=1, Translation="2,0,0", Rotation="0,0,0", Scale="1,1,1") {
    rotationMode=BMM_Additive
    rotationSpace=BCS_BoneSpace
    translationMode=BMM_Additive
    translationSpace=BCS_BoneSpace
    alphaBoolBlend {blendOption=Linear}
    boneToModify {boneName=dyn_straps_bk_02}
    translation {x=2}
   }
   n39 = Anim Dynamics - Bone: dyn_shoestrap_01_l(ComponentPose=@n38, Alpha=1, ExternalForce="0,0,0") {
    angularDampingOverride=1
    angularSpringConstant=1500
    bAngularSpring=true
    bGravityOverrideInSimSpace=true
    bOverrideAngularDamping=true
    gravityScale=0.35
    simulationSpace=RootRelative
    alphaBoolBlend {blendOption=Linear}
    boundBone {boneName=dyn_shoestrap_01_l}
    physicsBodyDefinitions [{
      sphereCollisionRadius=0
      boundBone {boneName=dyn_shoestrap_01_l}
      boxExtents=<10 20 20>
      constraintSetup {angularLimitsMax=<0 0 0>, angularLimitsMin=<0 0 -40>, angularTarget=<0 0 0>}
      localJointOffset=<10 0 0>
     }]
   }
   n40 = Anim Dynamics - Bone: dyn_shoestrap_02_l(ComponentPose=@n39, Alpha=1, ExternalForce="0,0,0") {
    angularDampingOverride=1
    angularSpringConstant=1500
    bAngularSpring=true
    bGravityOverrideInSimSpace=true
    bOverrideAngularDamping=true
    gravityScale=0.35
    simulationSpace=RootRelative
    alphaBoolBlend {blendOption=Linear}
    boundBone {boneName=dyn_shoestrap_02_l}
    componentAppliedLinearAccClamp=<3000 3000 3000>
    componentLinearAccScale=<0.3 0.3 0.3>
    componentLinearVelScale=<0.3 0.3 0.3>
    physicsBodyDefinitions [{
      sphereCollisionRadius=0
      boundBone {boneName=dyn_shoestrap_02_l}
      boxExtents=<10 20 20>
      constraintSetup {angularLimitsMax=<0 0 5>, angularLimitsMin=<0 0 -25>, angularTarget=<0 0 0>}
      localJointOffset=<10 0 0>
     }]
   }
   n41 = Anim Dynamics - Bone: dyn_shoelaceB_01_r(ComponentPose=@n40, Alpha=1, ExternalForce="0,0,0") {
    angularDampingOverride=1
    angularSpringConstant=1500
    bAngularSpring=true
    bGravityOverrideInSimSpace=true
    bOverrideAngularDamping=true
    gravityScale=0.35
    simulationSpace=RootRelative
    alphaBoolBlend {blendOption=Linear}
    boundBone {boneName=dyn_shoelaceB_01_r}
    physicsBodyDefinitions [{
      sphereCollisionRadius=0
      boundBone {boneName=dyn_shoelaceB_01_r}
      boxExtents=<10 20 20>
      constraintSetup {angularLimitsMax=<2.5 5 40>, angularLimitsMin=<-2.5 -5 -40>, angularTarget=<0 0 0>}
      localJointOffset=<10 0 0>
     }]
   }
   n42 = Anim Dynamics - Bone: dyn_shoelaceB_02_r(ComponentPose=@n41, Alpha=1, ExternalForce="0,0,0") {
    angularDampingOverride=1
    angularSpringConstant=1500
    bAngularSpring=true
    bGravityOverrideInSimSpace=true
    bOverrideAngularDamping=true
    gravityScale=0.35
    simulationSpace=RootRelative
    alphaBoolBlend {blendOption=Linear}
    boundBone {boneName=dyn_shoelaceB_02_r}
    physicsBodyDefinitions [{
      sphereCollisionRadius=0
      boundBone {boneName=dyn_shoelaceB_02_r}
      boxExtents=<10 20 20>
      constraintSetup {angularLimitsMax=<0 0 20>, angularLimitsMin=<0 -45 -20>, angularTarget=<0 0 0>}
      localJointOffset=<10 0 0>
     }]
   }
   n43 = Anim Dynamics - Bone: dyn_shoelaceA_01_r(ComponentPose=@n42, Alpha=1, ExternalForce="0,0,0") {
    angularDampingOverride=1
    angularSpringConstant=1500
    bAngularSpring=true
    bGravityOverrideInSimSpace=true
    bOverrideAngularDamping=true
    gravityScale=0.35
    simulationSpace=RootRelative
    alphaBoolBlend {blendOption=Linear}
    boundBone {boneName=dyn_shoelaceA_01_r}
    physicsBodyDefinitions [{
      sphereCollisionRadius=0
      boundBone {boneName=dyn_shoelaceA_01_r}
      boxExtents=<10 20 20>
      constraintSetup {angularLimitsMax=<2.5 5 40>, angularLimitsMin=<-2.5 -5 -40>, angularTarget=<0 0 0>}
      localJointOffset=<10 0 0>
     }]
   }
   n44 = Anim Dynamics - Bone: dyn_shoelaceA_02_r(ComponentPose=@n43, Alpha=1, ExternalForce="0,0,0") {
    angularDampingOverride=1
    angularSpringConstant=1500
    bAngularSpring=true
    bGravityOverrideInSimSpace=true
    bOverrideAngularDamping=true
    gravityScale=0.35
    simulationSpace=RootRelative
    alphaBoolBlend {blendOption=Linear}
    boundBone {boneName=dyn_shoelaceA_02_r}
    physicsBodyDefinitions [{
      sphereCollisionRadius=0
      boundBone {boneName=dyn_shoelaceA_02_r}
      boxExtents=<10 20 20>
      constraintSetup {angularLimitsMax=<2.5 0 20>, angularLimitsMin=<-2.5 -45 -20>, angularTarget=<0 0 0>}
      localJointOffset=<10 0 0>
     }]
   }
   Output Pose(Result=Component To Local(ComponentPose=@n44)) {layerGroup=None}
  ]}
 EventGraph {nodes [
   Event BlueprintUpdateAnimation()
   > n1 = Sequence()
   n2 = BreakTransform(InTransform=GetDeltaTransformFromRefPose(self=GetOwnerSkeletalMeshComponent(), BoneName=thigh_l, BaseName=pelvis))
   > Set thighCrouchDelta_slide_l(thighCrouchDelta_slide_l:double=MapRangeClamped(Value=BreakRotator(InRot=@n2.Rotation), InRangeA=-30, InRangeB=0, OutRangeA=1, OutRangeB=0)) -> (Output_Get:double)
   n3 = "Absolute (Float)"(A=BreakRotator(InRot=@n2.Rotation).Yaw)
   > Set thighCrouchDelta_l(thighCrouchDelta_l:double=SelectFloat(A=MapRangeClamped(Value=@n3, InRangeA=50, InRangeB=110), B=0, bPickA=Get bIsCrouching())) -> (Output_Get:double)
   n4 = Break Fort Anim Input Skydiving External Force(FortAnimInput_SkydivingExternalForce=Get SkydivingExternalForceInput()) -> (
    bUseSkydivingVectorForce=~
    SkydivingVectorForce=~
    HeadToPelvisDirection=~
    FloatingMultiplier=~
    FloatingAdditive=~
    DivingMultiplier=~
    DivingAdditive=~
    ParachutingMultiplier=~
    ParachutingAdditive=~
   )
   Set windVector(
    execute=@n1.then_1
    windVector:Vector=SelectVector(A=vector * float(A=@n4.SkydivingVectorForce, B=0.15), B="0, 0,-980", bPickA=Get bIsSkyDiving())
   ) -> (Output_Get:Vector)
   Event BlueprintInitializeAnimation()
   > Set isCrouchingAlpha(isCrouchingAlpha:double=0) -> (Output_Get:double)
   > Set isNotCrouchingAlpha(isNotCrouchingAlpha:double=1) -> (Output_Get:double)
   Event OnBeginCrouching()
   > Set isCrouchingAlpha(isCrouchingAlpha=1)
   > Set isNotCrouchingAlpha(isNotCrouchingAlpha=0)
   Event OnEndCrouching()
   > Set isCrouchingAlpha(isCrouchingAlpha=0)
   > Set isNotCrouchingAlpha(isNotCrouchingAlpha=1)
  ]}
]
variables [
 isCrouchingAlpha {type=number}
 isNotCrouchingAlpha {type=number}
 thighCrouchDelta_l {type=number}
 thighCrouchDelta_slide_l {type=number}
 windVector {type=Vector}
]
