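# Animation Blueprint for the face accessory part of the female Tech Ops Blue character, adding secondary motion to a left antenna.
# It copies the pose from the attached parent mesh, then runs an Anim Dynamics chain over the two antenna bones at partial alpha.
# The simulation is root-relative, with angular springs, tight angular limits and a gravity override.
# A Trail controller for the antenna tip is left unconnected, and the event graph holds only unwired stubs.
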AnimBlueprint /BRCosmetics/Characters/Player/Female/Medium/Bodies/F_MED_Tech_Ops_Blue/Meshes/Parts/F_MED_Tech_Ops_Blue_FaceAcc_AnimBP
target_skeleton=&/F_MED_Tech_Ops_Blue_FaceAcc_Skeleton
graphs [
 AnimGraph {nodes [
   n1 = "Anim Dynamics - Chain: dyn_antenna_01_l -> dyn_antenna_02_l"!(
    ComponentPose=Local To Component(LocalPose=Copy Pose From Mesh() {bUseAttachedParent=true})
    Alpha=0.4
    GravityOverride:Vector="0,0,0"
    ExternalForce="0,0,0"
   ) -> (Pose) {
    angularDampingOverride=1
    angularSpringConstant=1500
    bAngularSpring=true
    bChain=true
    bGravityOverrideInSimSpace=true
    bOverrideAngularDamping=true
    bUseGravityOverride=true
    gravityScale=0.35
    simulationSpace=RootRelative
    alphaBoolBlend {blendOption=Linear}
    boundBone {boneName=dyn_antenna_01_l}
    chainEnd {boneName=dyn_antenna_02_l}
    componentAppliedLinearAccClamp=<3000 3000 3000>
    componentLinearAccScale=<0.3 0.3 0.3>
    componentLinearVelScale=<0.6 0.6 0.6>
    gravityOverride {z=-980}
    physicsBodyDefinitions [{
      sphereCollisionRadius=0
      boundBone {boneName=dyn_antenna_01_l}
      boxExtents=<20 10 10>
      constraintSetup {angularLimitsMax=<5 5 5>, angularLimitsMin=<-5 -5 -5>, angularTarget=<0 0 0>}
      localJointOffset=<5 0 0>
     }]
   }
   Output Pose(Result=Component To Local(ComponentPose=@n1)) {layerGroup=None}
   Trail controller - Bone: dyn_antenna_02_l!(ComponentPose, Alpha=0.2, RelaxationSpeedScale:float=10) -> (Pose) {
    maxDeltaTime=0.0333
    relaxationSpeedScale=4
    rotationLimits [{limitMax=<180 5 180>, limitMin=<-180 -5 -180>}, {limitMax=<180 5 180>, limitMin=<-180 -5 -180>}]
    rotationOffsets [<0 0 0>, <0 0 0>]
    trailBone {boneName=dyn_antenna_02_l}
   }
  ]}
 EventGraph {nodes [
   TryGetPawnOwner()
   Event BlueprintUpdateAnimation()
  ]}
]
variables []
